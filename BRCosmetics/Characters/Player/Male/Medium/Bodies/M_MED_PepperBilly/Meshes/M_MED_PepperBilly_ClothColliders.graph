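# Physics asset for the PepperBilly medium male player body, named as the collider set for its cloth simulation.
# Only pelvis, spine_01 and spine_03 carry collision, one capsule each, forming a torso volume.
# Root, thigh, calf and foot bodies exist but have no shapes.
# Constraints chain spine, pelvis, root and both legs, with all swing and twist axes free.
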PhysicsAsset /BRCosmetics/Characters/Player/Male/Medium/Bodies/M_MED_PepperBilly/Meshes/M_MED_PepperBilly_ClothColliders
bodies {
 calf_l []
 calf_r []
 foot_l []
 foot_r []
 pelvis [{shapeType=Capsule, length=11.1814, radius=12.0029, center=<-2.4887 -0.15 0>}]
 root []
 spine_01 [{shapeType=Capsule, length=10.1059, radius=11.3964, center=<-1.7965 0.1884 0>}]
 spine_03 [{shapeType=Capsule, length=7.8105, radius=11.8132, center=<0.1445 0.2396 0.0001>}]
 thigh_l []
 thigh_r []
}
constraints [
 {
  bone1Name=spine_01
  bone2Name=pelvis
  swing1LimitDegrees=45
  swing1Motion=Free
  swing2LimitDegrees=45
  swing2Motion=Free
  twistLimitDegrees=45
  twistMotion=Free
 }
 {
  bone1Name=spine_03
  bone2Name=spine_01
  swing1LimitDegrees=45
  swing1Motion=Free
  swing2LimitDegrees=45
  swing2Motion=Free
  twistLimitDegrees=45
  twistMotion=Free
 }
 {
  bone1Name=thigh_l
  bone2Name=pelvis
  swing1LimitDegrees=45
  swing1Motion=Free
  swing2LimitDegrees=45
  swing2Motion=Free
  twistLimitDegrees=45
  twistMotion=Free
 }
 {
  bone1Name=foot_l
  bone2Name=calf_l
  swing1LimitDegrees=45
  swing1Motion=Free
  swing2LimitDegrees=45
  swing2Motion=Free
  twistLimitDegrees=45
  twistMotion=Free
 }
 {
  bone1Name=foot_r
  bone2Name=calf_r
  swing1LimitDegrees=45
  swing1Motion=Free
  swing2LimitDegrees=45
  swing2Motion=Free
  twistLimitDegrees=45
  twistMotion=Free
 }
 {
  bone1Name=calf_r
  bone2Name=thigh_r
  swing1LimitDegrees=45
  swing1Motion=Free
  swing2LimitDegrees=45
  swing2Motion=Free
  twistLimitDegrees=45
  twistMotion=Free
 }
 {
  bone1Name=thigh_r
  bone2Name=pelvis
  swing1LimitDegrees=45
  swing1Motion=Free
  swing2LimitDegrees=45
  swing2Motion=Free
  twistLimitDegrees=45
  twistMotion=Free
 }
 {
  bone1Name=calf_l
  bone2Name=thigh_l
  swing1LimitDegrees=45
  swing1Motion=Free
  swing2LimitDegrees=45
  swing2Motion=Free
  twistLimitDegrees=45
  twistMotion=Free
 }
 {
  bone1Name=pelvis
  bone2Name=root
  swing1LimitDegrees=45
  swing1Motion=Free
  swing2LimitDegrees=45
  swing2Motion=Free
  twistLimitDegrees=45
  twistMotion=Free
 }
]
