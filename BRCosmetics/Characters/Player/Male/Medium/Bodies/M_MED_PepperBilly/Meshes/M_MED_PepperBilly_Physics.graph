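# Physics asset for the PepperBilly medium male body cosmetic, set up for dynamic cloth bones hanging from the pelvis.
# Capsule chains at the back and three per side use mostly tight swing and twist limits, with cross-constraints linking neighbouring side chains lower down.
# Pelvis, leg and foot capsules form the lower body, and an oversized root capsule sits below the origin like a ground collider.
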PhysicsAsset /BRCosmetics/Characters/Player/Male/Medium/Bodies/M_MED_PepperBilly/Meshes/M_MED_PepperBilly_Physics
bodies {
 calf_l [{shapeType=Capsule, length=45.8272, radius=8.3373, center=<-27.4545 -1.3118 0.4701>, rotation=<rot -0.0944 84.2262 -88.777>}]
 calf_r [{shapeType=Capsule, length=43.9744, radius=10.3549, center=<27.3794 -0.1258 -0.3504>, rotation=<rot -0.1656 84.4618 -88.3189>}]
 dyn_Cloth_volume_bk [{shapeType=Capsule, length=12.5997, radius=8.3151, shapeName=dyn_Cloth_volume_bk_capsule, center=<-9.3613 -3.4677 0>}]
 dyn_Cloth_volume_bk01 [{shapeType=Capsule, length=12.5997, radius=8.3151, shapeName=dyn_Cloth_volume_bk01_capsule, center=<-9.3613 -3.4677 0>}]
 dyn_Cloth_volume_bk02 [{shapeType=Capsule, length=12.5997, radius=8.3151, shapeName=dyn_Cloth_volume_bk02_capsule, center=<-9.3613 -3.4677 0>}]
 dyn_Cloth_volume_l_A [{shapeType=Capsule, length=12.5997, radius=8.3151, shapeName=dyn_Cloth_volume_l_A_capsule, center=<-9.3613 -3.4677 0>}]
 dyn_Cloth_volume_l_A01 [{shapeType=Capsule, length=12.5997, radius=8.3151, shapeName=dyn_Cloth_volume_l_A01_capsule, center=<-9.3613 -3.4677 0>}]
 dyn_Cloth_volume_l_A02 [{shapeType=Capsule, length=12.5997, radius=8.3151, shapeName=dyn_Cloth_volume_l_A02_capsule, center=<-9.3613 -3.4677 0>}]
 dyn_Cloth_volume_l_A03 [{shapeType=Capsule, length=12.5997, radius=8.3151, shapeName=dyn_Cloth_volume_l_A03_capsule, center=<-9.3613 -3.4677 0>}]
 dyn_Cloth_volume_l_A04 [{shapeType=Capsule, length=9.8497, radius=8.3151, shapeName=dyn_Cloth_volume_l_A04_capsule, center=<-9.3613 -3.4677 0>}]
 dyn_Cloth_volume_l_B [{shapeType=Capsule, length=12.5997, radius=8.3151, shapeName=dyn_Cloth_volume_l_B_capsule, center=<-9.3613 -3.4677 0>}]
 dyn_Cloth_volume_l_B01 [{shapeType=Capsule, length=12.5997, radius=8.3151, shapeName=dyn_Cloth_volume_l_B01_capsule, center=<-9.3613 -3.4677 0>}]
 dyn_Cloth_volume_l_B02 [{shapeType=Capsule, length=12.5997, radius=8.3151, shapeName=dyn_Cloth_volume_l_B02_capsule, center=<-9.3613 -3.4677 0>}]
 dyn_Cloth_volume_l_B03 [{shapeType=Capsule, length=12.5997, radius=8.3151, shapeName=dyn_Cloth_volume_l_B03_capsule, center=<-9.3613 -3.4677 0>}]
 dyn_Cloth_volume_l_B04 [{shapeType=Capsule, length=8.0997, radius=8.3151, shapeName=dyn_Cloth_volume_l_B04_capsule, center=<-9.3613 -3.4677 0>}]
 dyn_Cloth_volume_l_C [{shapeType=Capsule, length=12.5997, radius=8.3151, shapeName=dyn_Cloth_volume_l_C_capsule, center=<-9.3613 -3.4677 0>}]
 dyn_Cloth_volume_l_C01 [{shapeType=Capsule, length=12.236, radius=8.6788, shapeName=dyn_Cloth_volume_l_C01_capsule, center=<-11.6607 -10.2614 0>}]
 dyn_Cloth_volume_l_C02 [{shapeType=Capsule, length=3.8408, radius=11.4254, shapeName=dyn_Cloth_volume_l_C02_capsule, center=<-11.6607 -8.6201 0>}]
 dyn_Cloth_volume_l_C03 [{shapeType=Capsule, length=4.6073, radius=11.4254, shapeName=dyn_Cloth_volume_l_C03_capsule, center=<-11.6607 -6.6798 0>}]
 dyn_Cloth_volume_l_C04 [{shapeType=Capsule, length=4.6073, radius=11.4254, shapeName=dyn_Cloth_volume_l_C04_capsule, center=<-11.6607 -6.6798 0>}]
 dyn_Cloth_volume_r_A [{shapeType=Capsule, length=12.5997, radius=8.3151, shapeName=dyn_Cloth_volume_r_A_capsule, center=<-9.3613 -3.4677 0>}]
 dyn_Cloth_volume_r_A01 [{shapeType=Capsule, length=12.5997, radius=8.3151, shapeName=dyn_Cloth_volume_r_A01_capsule, center=<-9.3613 -3.4677 0>}]
 dyn_Cloth_volume_r_A02 [{shapeType=Capsule, length=12.5997, radius=8.3151, shapeName=dyn_Cloth_volume_r_A02_capsule, center=<-9.3613 -4.7758 0>}]
 dyn_Cloth_volume_r_A03 [{shapeType=Capsule, length=12.5997, radius=8.3151, shapeName=dyn_Cloth_volume_r_A03_capsule, center=<-9.3613 -4.6921 0>}]
 dyn_Cloth_volume_r_A04 [{shapeType=Capsule, length=12.5997, radius=8.3151, shapeName=dyn_Cloth_volume_r_A04_capsule, center=<-9.3613 -4.029 0>}]
 dyn_Cloth_volume_r_B [{shapeType=Capsule, length=12.5997, radius=8.3151, shapeName=dyn_Cloth_volume_r_B_capsule, center=<-9.3613 -3.4677 0>}]
 dyn_Cloth_volume_r_B01 [{shapeType=Capsule, length=12.5997, radius=8.3151, shapeName=dyn_Cloth_volume_r_B01_capsule, center=<-9.3613 -3.4677 0>}]
 dyn_Cloth_volume_r_B02 [{shapeType=Capsule, length=12.5997, radius=8.3151, shapeName=dyn_Cloth_volume_r_B02_capsule, center=<-9.3613 -3.4677 0>}]
 dyn_Cloth_volume_r_B03 [{shapeType=Capsule, length=12.5997, radius=8.3151, shapeName=dyn_Cloth_volume_r_B03_capsule, center=<-9.3613 -3.4677 0>}]
 dyn_Cloth_volume_r_B04 [{shapeType=Capsule, length=12.5997, radius=8.3151, shapeName=dyn_Cloth_volume_r_B04_capsule, center=<-9.3613 -3.4677 0>}]
 dyn_Cloth_volume_r_C [{shapeType=Capsule, length=12.5997, radius=8.3151, shapeName=dyn_Cloth_volume_r_C_capsule, center=<-9.3613 -3.4677 0>}]
 dyn_Cloth_volume_r_C01 [{shapeType=Capsule, length=12.236, radius=8.6788, shapeName=dyn_Cloth_volume_r_C01_capsule, center=<-11.6607 -10.2614 0>}]
 dyn_Cloth_volume_r_C02 [{shapeType=Capsule, length=3.8408, radius=11.4254, shapeName=dyn_Cloth_volume_r_C02_capsule, center=<-11.6607 -8.6201 0>}]
 dyn_Cloth_volume_r_C03 [{shapeType=Capsule, length=4.6073, radius=11.4254, shapeName=dyn_Cloth_volume_r_C03_capsule, center=<-11.6607 -6.6798 0>}]
 dyn_Cloth_volume_r_C04 [{shapeType=Capsule, length=4.6073, radius=11.4254, shapeName=dyn_Cloth_volume_r_C04_capsule, center=<-11.6607 -6.6798 0>}]
 foot_l [{shapeType=Capsule, length=9.6416, radius=13.5721, shapeName=foot_l_capsule, center=<-2.737 6.0615 0.0778>, rotation=<rot 0 -3.4949 -84.5306>}]
 foot_r [{
   shapeType=Capsule
   length=9.6416
   radius=13.5721
   shapeName=foot_r_capsule
   center=<2.737 -6.0615 -0.0778>
   rotation=<rot 0 -3.4949 -84.5306>
  }]
 pelvis [{shapeType=Capsule, length=11.4921, radius=9.9478, center=<1.9277 -0.3348 0.4765>, rotation=<rot -3.4833 -90.0002 177.7238>}]
 root [{
   shapeType=Capsule
   length=52.1648
   radius=105.75
   shapeName=dyn_Cloth_volume_l_B02_capsule
   center=<0 -34.1383 -104.8952>
   rotation=<rot 90 0 0>
  }]
 thigh_l [{shapeType=Capsule, length=28.6432, radius=8.2187, center=<-23.1014 -0.0511 1.2559>, rotation=<rot 0.0376 -86.592 -90.72>}]
 thigh_r [{shapeType=Capsule, length=27.2268, radius=9.5683, center=<22.9845 2.1157 -1.0594>, rotation=<rot -0.0699 -85.064 -88.8823>}]
}
constraints [
 {bone1Name=thigh_l, bone2Name=pelvis, swing1LimitDegrees=45, swing2LimitDegrees=45, twistLimitDegrees=45}
 {bone1Name=calf_l, bone2Name=thigh_l, swing1LimitDegrees=45, swing2LimitDegrees=45, twistLimitDegrees=45}
 {bone1Name=thigh_r, bone2Name=pelvis, swing1LimitDegrees=45, swing2LimitDegrees=45, twistLimitDegrees=45}
 {bone1Name=calf_r, bone2Name=thigh_r, swing1LimitDegrees=45, swing2LimitDegrees=45, twistLimitDegrees=45}
 {bone1Name=dyn_Cloth_volume_bk, bone2Name=pelvis, swing1LimitDegrees=45, swing2LimitDegrees=45, twistLimitDegrees=45}
 {bone1Name=dyn_Cloth_volume_bk01, bone2Name=dyn_Cloth_volume_bk, swing1LimitDegrees=45, swing2LimitDegrees=15, twistLimitDegrees=10}
 {bone1Name=dyn_Cloth_volume_bk02, bone2Name=dyn_Cloth_volume_bk01, swing1LimitDegrees=45, swing2LimitDegrees=15, twistLimitDegrees=10}
 {bone1Name=dyn_Cloth_volume_l_A, bone2Name=pelvis, swing1LimitDegrees=45, swing2LimitDegrees=15, twistLimitDegrees=10}
 {bone1Name=dyn_Cloth_volume_l_A01, bone2Name=dyn_Cloth_volume_l_A, swing1LimitDegrees=45, swing2LimitDegrees=15, twistLimitDegrees=10}
 {bone1Name=dyn_Cloth_volume_l_A02, bone2Name=dyn_Cloth_volume_l_A01, swing1LimitDegrees=45, swing2LimitDegrees=15, twistLimitDegrees=10}
 {bone1Name=dyn_Cloth_volume_l_A03, bone2Name=dyn_Cloth_volume_l_A02, swing1LimitDegrees=15, swing2LimitDegrees=15, twistLimitDegrees=10}
 {bone1Name=dyn_Cloth_volume_l_A04, bone2Name=dyn_Cloth_volume_l_A03, swing1LimitDegrees=0, swing2LimitDegrees=5, twistLimitDegrees=5}
 {bone1Name=dyn_Cloth_volume_l_B, bone2Name=pelvis, swing1LimitDegrees=45, swing2LimitDegrees=15, twistLimitDegrees=10}
 {bone1Name=dyn_Cloth_volume_l_B01, bone2Name=dyn_Cloth_volume_l_B, swing1LimitDegrees=20, swing2LimitDegrees=15, twistLimitDegrees=10}
 {bone1Name=dyn_Cloth_volume_l_B02, bone2Name=dyn_Cloth_volume_l_B01, swing1LimitDegrees=20, swing2LimitDegrees=15, twistLimitDegrees=10}
 {bone1Name=dyn_Cloth_volume_l_B03, bone2Name=dyn_Cloth_volume_l_B02, swing1LimitDegrees=45, swing2LimitDegrees=15, twistLimitDegrees=10}
 {bone1Name=dyn_Cloth_volume_l_B04, bone2Name=dyn_Cloth_volume_l_B03, swing1LimitDegrees=45, swing2LimitDegrees=15, twistLimitDegrees=10}
 {bone1Name=dyn_Cloth_volume_l_C, bone2Name=pelvis, swing1LimitDegrees=45, swing2LimitDegrees=15, twistLimitDegrees=10}
 {bone1Name=dyn_Cloth_volume_l_C01, bone2Name=dyn_Cloth_volume_l_C, swing1LimitDegrees=20, swing2LimitDegrees=15, twistLimitDegrees=10}
 {bone1Name=dyn_Cloth_volume_l_C02, bone2Name=dyn_Cloth_volume_l_C01, swing1LimitDegrees=30, swing2LimitDegrees=15, twistLimitDegrees=10}
 {bone1Name=dyn_Cloth_volume_l_C03, bone2Name=dyn_Cloth_volume_l_C02, swing1LimitDegrees=45, swing2LimitDegrees=15, twistLimitDegrees=10}
 {bone1Name=dyn_Cloth_volume_l_C04, bone2Name=dyn_Cloth_volume_l_C03, swing1LimitDegrees=45, swing2LimitDegrees=15, twistLimitDegrees=10}
 {bone1Name=dyn_Cloth_volume_r_A, bone2Name=pelvis, swing1LimitDegrees=45, swing2LimitDegrees=15, twistLimitDegrees=10}
 {bone1Name=dyn_Cloth_volume_r_A01, bone2Name=dyn_Cloth_volume_r_A, swing1LimitDegrees=45, swing2LimitDegrees=15, twistLimitDegrees=10}
 {bone1Name=dyn_Cloth_volume_r_A02, bone2Name=dyn_Cloth_volume_r_A01, swing1LimitDegrees=45, swing2LimitDegrees=15, twistLimitDegrees=10}
 {bone1Name=dyn_Cloth_volume_r_A03, bone2Name=dyn_Cloth_volume_r_A02, swing1LimitDegrees=15, swing2LimitDegrees=15, twistLimitDegrees=10}
 {bone1Name=dyn_Cloth_volume_r_A04, bone2Name=dyn_Cloth_volume_r_A03, swing1LimitDegrees=10, swing2LimitDegrees=15, twistLimitDegrees=10}
 {bone1Name=dyn_Cloth_volume_r_B, bone2Name=pelvis, swing1LimitDegrees=45, swing2LimitDegrees=15, twistLimitDegrees=10}
 {bone1Name=dyn_Cloth_volume_r_B01, bone2Name=dyn_Cloth_volume_r_B, swing1LimitDegrees=20, swing2LimitDegrees=15, twistLimitDegrees=10}
 {bone1Name=dyn_Cloth_volume_r_B02, bone2Name=dyn_Cloth_volume_r_B01, swing1LimitDegrees=20, swing2LimitDegrees=15, twistLimitDegrees=10}
 {bone1Name=dyn_Cloth_volume_r_B03, bone2Name=dyn_Cloth_volume_r_B02, swing1LimitDegrees=45, swing2LimitDegrees=15, twistLimitDegrees=10}
 {bone1Name=dyn_Cloth_volume_r_B04, bone2Name=dyn_Cloth_volume_r_B03, swing1LimitDegrees=45, swing2LimitDegrees=15, twistLimitDegrees=10}
 {bone1Name=dyn_Cloth_volume_r_C, bone2Name=pelvis, swing1LimitDegrees=45, swing2LimitDegrees=15, twistLimitDegrees=10}
 {bone1Name=dyn_Cloth_volume_r_C01, bone2Name=dyn_Cloth_volume_r_C, swing1LimitDegrees=20, swing2LimitDegrees=15, twistLimitDegrees=10}
 {bone1Name=dyn_Cloth_volume_r_C02, bone2Name=dyn_Cloth_volume_r_C01, swing1LimitDegrees=30, swing2LimitDegrees=15, twistLimitDegrees=10}
 {bone1Name=dyn_Cloth_volume_r_C03, bone2Name=dyn_Cloth_volume_r_C02, swing1LimitDegrees=45, swing2LimitDegrees=15, twistLimitDegrees=10}
 {bone1Name=dyn_Cloth_volume_r_C04, bone2Name=dyn_Cloth_volume_r_C03, swing1LimitDegrees=45, swing2LimitDegrees=15, twistLimitDegrees=10}
 {
  bone1Name=dyn_Cloth_volume_r_B04
  bone2Name=dyn_Cloth_volume_r_C04
  swing1LimitDegrees=45
  swing1Motion=Free
  swing2LimitDegrees=45
  swing2Motion=Free
  twistLimitDegrees=15
 }
 {
  bone1Name=dyn_Cloth_volume_l_B04
  bone2Name=dyn_Cloth_volume_l_C04
  swing1LimitDegrees=45
  swing1Motion=Free
  swing2LimitDegrees=45
  swing2Motion=Free
  twistLimitDegrees=45
  twistMotion=Free
 }
 {bone1Name=dyn_Cloth_volume_l_A04, bone2Name=dyn_Cloth_volume_l_B04, swing1LimitDegrees=45, swing2LimitDegrees=45, twistLimitDegrees=45}
 {
  bone1Name=pelvis
  bone2Name=root
  swing1LimitDegrees=45
  swing1Motion=Free
  swing2LimitDegrees=45
  swing2Motion=Free
  twistLimitDegrees=45
  twistMotion=Free
 }
 {bone1Name=dyn_Cloth_volume_l_A03, bone2Name=dyn_Cloth_volume_l_B03, swing1LimitDegrees=45, swing2LimitDegrees=45, twistLimitDegrees=45}
 {
  bone1Name=dyn_Cloth_volume_l_B03
  bone2Name=dyn_Cloth_volume_l_C03
  swing1LimitDegrees=45
  swing1Motion=Free
  swing2LimitDegrees=45
  swing2Motion=Free
  twistLimitDegrees=45
  twistMotion=Free
 }
 {
  bone1Name=dyn_Cloth_volume_r_B03
  bone2Name=dyn_Cloth_volume_r_C03
  swing1LimitDegrees=45
  swing1Motion=Free
  swing2LimitDegrees=45
  swing2Motion=Free
  twistLimitDegrees=15
 }
 {
  bone1Name=dyn_Cloth_volume_r_B02
  bone2Name=dyn_Cloth_volume_r_A02
  swing1LimitDegrees=45
  swing1Motion=Free
  swing2LimitDegrees=45
  swing2Motion=Free
  twistLimitDegrees=45
  twistMotion=Free
 }
 {
  bone1Name=dyn_Cloth_volume_r_A04
  bone2Name=dyn_Cloth_volume_r_B04
  swing1LimitDegrees=45
  swing1Motion=Free
  swing2LimitDegrees=45
  swing2Motion=Free
  twistLimitDegrees=45
  twistMotion=Free
 }
 {
  bone1Name=dyn_Cloth_volume_r_A03
  bone2Name=dyn_Cloth_volume_r_B03
  swing1LimitDegrees=45
  swing1Motion=Free
  swing2LimitDegrees=45
  swing2Motion=Free
  twistLimitDegrees=45
  twistMotion=Free
 }
 {
  bone1Name=foot_r
  bone2Name=calf_r
  swing1LimitDegrees=45
  swing1Motion=Free
  swing2LimitDegrees=45
  swing2Motion=Free
  twistLimitDegrees=45
  twistMotion=Free
 }
 {
  bone1Name=foot_l
  bone2Name=calf_l
  swing1LimitDegrees=45
  swing1Motion=Free
  swing2LimitDegrees=45
  swing2Motion=Free
  twistLimitDegrees=45
  twistMotion=Free
 }
]
