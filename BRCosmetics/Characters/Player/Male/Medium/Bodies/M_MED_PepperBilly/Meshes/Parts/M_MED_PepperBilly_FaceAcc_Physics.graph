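# Physics asset for the face accessory mesh of the PepperBilly medium male player cosmetic, defining collision and joint limits for its two dynamic ears.
# Each ear is a two-bone chain of small, angled, offset capsules, anchored to a larger zero-length capsule on the head that acts as a sphere.
# All four joints share the same tight swing and twist limits, with the second swing axis the loosest.
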PhysicsAsset /BRCosmetics/Characters/Player/Male/Medium/Bodies/M_MED_PepperBilly/Meshes/Parts/M_MED_PepperBilly_FaceAcc_Physics
bodies {
 dyn_ear_L_01 [{shapeType=Capsule, length=2, radius=2.25, shapeName=dyn_ear_L_01_capsule, center=<0.5634 0 1.5747>, rotation=<rot -19.6875 0.0002 0>}]
 dyn_ear_L_02 [{shapeType=Capsule, length=2, radius=2.25, shapeName=dyn_ear_L_02_capsule, center=<0.5634 0 1.5747>, rotation=<rot -19.6874 0.0001 0>}]
 dyn_ear_R_01 [{shapeType=Capsule, length=2, radius=2.25, shapeName=dyn_ear_R_01_capsule, center=<1.0491 0 -2.2182>, rotation=<rot 25.3125 0.0004 0>}]
 dyn_ear_R_02 [{shapeType=Capsule, length=2, radius=2.25, shapeName=dyn_ear_R_02_capsule, center=<1.0491 0 -2.2182>, rotation=<rot 25.3126 0.0003 0>}]
 head [{shapeType=Capsule, length=0, radius=8.2748}]
}
constraints [
 {bone1Name=dyn_ear_L_01, bone2Name=head, swing1LimitDegrees=5, swing2LimitDegrees=25, twistLimitDegrees=10}
 {bone1Name=dyn_ear_L_02, bone2Name=dyn_ear_L_01, swing1LimitDegrees=5, swing2LimitDegrees=25, twistLimitDegrees=10}
 {bone1Name=dyn_ear_R_01, bone2Name=head, swing1LimitDegrees=5, swing2LimitDegrees=25, twistLimitDegrees=10}
 {bone1Name=dyn_ear_R_02, bone2Name=dyn_ear_R_01, swing1LimitDegrees=5, swing2LimitDegrees=25, twistLimitDegrees=10}
]
